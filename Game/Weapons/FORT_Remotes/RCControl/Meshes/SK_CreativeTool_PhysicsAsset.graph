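# Physics asset for the creative tool skeletal mesh, a phone-style RC remote weapon in the FORT_Remotes set.
# It holds a single body on the Phone_Offset bone, built from one capsule laid on its side slightly off the bone origin.
# It has no constraints, so the asset provides one simple collision shape with no jointed parts.
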PhysicsAsset /Game/Weapons/FORT_Remotes/RCControl/Meshes/SK_CreativeTool_PhysicsAsset
bodies {Phone_Offset [{shapeType=Capsule, length=8.7549, radius=4.0462, center=<0.7879 -0.0128 -0.9173>, rotation=<rot 0.2518 -90 0>}]}
constraints []
